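; HuskInterface is a Blueprint interface declaring the messages that Husk enemies can receive for variant-specific behaviors.
; Propane tank functions cover state-machine hooks for melee start, failed spawn and removal, plus ignition with a pawn, remaining fuse time and drop flag.
; Other functions report a despawned bee cloud projectile and toggle a burp effect.
; It holds signatures only, with no logic.
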(fn BeeCloudDespawned
  (params (Cloud :type Fort%20Projectile%20Base%20Object%20Reference)))

(fn FSMBeginPropaneTankMelee)

(fn FSMFailedToSpawnTank)

(fn FSMRemovePropaneTank)

(fn IgnitePropaneTank
  (params (Pawn :type Fort%20Pawn%20Object%20Reference) (RemainingFuseTime :type Float%20%28double-precision%29) (JustDrop :type Boolean)))

(fn ManageBurpEffect
  (params (Enabled :type Boolean)))
