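# NS_Headshot is a Niagara system in the Husk enemy's effects folder, named as a headshot effect.
# It holds two emitters, Rings and XX, with no renderers listed for either.
# SystemState is the only module recorded, and the system has no user variables, dynamic inputs or data interfaces.
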NiagaraSystem /Game/Characters/Enemies/Husk/Effects/NS_Headshot
dependencies {usedDataInterfaces [], usedDynamicInputs [], usedModules [&SystemState]}
emitters {Rings {rendererClasses [], renderers []}, XX {rendererClasses [], renderers []}}
summary {systemName=NS_Headshot, emitters [{emitterName=Rings, rendererClasses []}, {emitterName=XX, rendererClasses []}], userVariables []}
