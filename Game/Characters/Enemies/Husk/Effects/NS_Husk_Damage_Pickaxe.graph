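# Niagara damage effect for the Husk enemy, named for pickaxe hits: a short, blue-violet mesh flash with a puff of smoke.
# Base bursts a randomly rolled, camera-offset mesh that grows and fades within about a tenth of a second.
# SmokeSwirl bursts a few forward-thrown, upward-accelerated, drag-slowed flipbook sprites that grow, rotate, darken and fade.
# Curves animate the smoke's dynamic material parameters, and Cascade converter modules appear in both emitters.
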NiagaraSystem /Game/Characters/Enemies/Husk/Effects/NS_Husk_Damage_Pickaxe
dependencies {
 usedDataInterfaces [
  &NiagaraDataInterfaceCurve
  &NiagaraDataInterfaceSpriteRendererInfo
  &NiagaraDataInterfaceVectorCurve
 ]
 usedDynamicInputs []
 usedModules [&SystemState]
}
emitters {
 Base {
  emitterUpdateScript [
   &EmitterState(Life Cycle Mode=::NewEnumerator1, Loop Behavior=::NewEnumerator1, Loop Duration=1, UseLoopDelay=true)
   &SpawnBurst_Instantaneous(Use Spawn Probability=false, UseLoopCountLimit=false)
  ]
  particleSpawnScript [
   /SetVariables(Particles.MeshRotationEuler:Vector3f=&RandomRangeVector(Minimum=<0 -0.02 -0.02>, Maximum=<1 0.02 0.02>, Evaluation Type=::NewEnumerator1)) {isSetParameters=true}
   /SetVariables(Particles.MeshOrientation=/CascadeToNiagaraConverter/NiagaraScripts/DynamicInputScripts/MakeQuatFromEuler(Euler:Vector3f=$Particles.MeshRotationEuler, Scale:NiagaraFloat=360)) {isSetParameters=true}
   &InitializeParticle(
    Lifetime=&RandomRangeFloat(Minimum=0.07, Maximum=0.1, Evaluation Type=::NewEnumerator1)
    Mesh Scale=&RandomRangeVector(Minimum=<1.5 0.8 0.8>, Maximum=<2 1 1>, Evaluation Type=::NewEnumerator1)
    Mesh Scale Mode=::NewEnumerator4
    UsePositionOffset=false
    Write Material Random=true
   )
   &CameraOffset(Camera Offset Amount=10)
  ]
  particleUpdateScript [
   &ParticleState()
   &ScaleMeshSize(Scale Factor=&VectorFromFloat(Value=&FloatFromCurve(FloatCurve=DataInterface() {propertyValues {Curve {keys [<at 0 0.5>,<at 1.0103 1 RCIM_Cubic>]}}})))
   &Color(Color=&MakeLinearColorFromVectorAndFloat(
     "Vector (RGB)"=<3.259 1.6 20>
     "Float (Alpha)"=&FloatFromCurve(FloatCurve=DataInterface() {propertyValues {Curve {keys [<at 0.35 1 RCIM_Cubic>,<at 1 0 RCIM_Cubic>]}}})
    ))
  ]
  rendererClasses [&NiagaraMeshRendererProperties]
  renderers [{rendererClass=&NiagaraMeshRendererProperties, rendererIndex=0}]
 }
 SmokeSwirl {
  emitterUpdateScript [
   &EmitterState(Life Cycle Mode=::NewEnumerator1, Loop Behavior=::NewEnumerator1, Loop Duration=1.4, UseLoopDelay=true)
   &SpawnBurst_Instantaneous(
    Spawn Count=&RandomRangeInt(Minimum=2, Maximum=4)
    Use Spawn Probability=false
    UseLoopCountLimit=false
   )
  ]
  particleSpawnScript [
   /SetVariables(Particles.SpriteRotationRate=&RandomRangeFloat(Minimum=-0.2, Maximum=0.2, Evaluation Type=::NewEnumerator1)) {isSetParameters=true}
   /SetVariables(Particles.Acceleration:Vector3f=&RandomRangeVector(Minimum=<0 0 50>, Maximum=<0 0 200>, Evaluation Type=::NewEnumerator1)) {isSetParameters=true}
   &InitializeParticle(
    Lifetime=&RandomRangeFloat(Minimum=0.3, Maximum=0.4, Evaluation Type=::NewEnumerator1)
    Sprite Rotation Angle=&RandomRangeFloat(Minimum=0, Maximum=1, Evaluation Type=::NewEnumerator1)
    Sprite Rotation Mode=::NewEnumerator1
    Sprite Size=&Vector2DFromFloat(Value=&RandomRangeFloat(Minimum=150, Maximum=200, Evaluation Type=::NewEnumerator1))
    Sprite Size Mode=::NewEnumerator4
    UsePositionOffset=false
    Write Material Random=true
   )
   &AddVelocity(Velocity=&RandomRangeVector(Minimum=<100 0 -100>, Maximum=<500 60 100>, Evaluation Type=::NewEnumerator1))
   &Color(Color=&MakeLinearColorFromVectorAndFloat(
     "Vector (RGB)"=<1 1 1>
     "Float (Alpha)"=&RandomRangeFloat(Minimum=0.3, Maximum=0.5, Evaluation Type=::NewEnumerator1)
    ))
   &ShapeLocation(MaskSpawnGroup=false, Override Seed=false, Sphere Radius=5, UseOwnerScale=true)
   /CascadeToNiagaraConverter/NiagaraScripts/ModuleScripts/FixShapeVector()
   &AddVelocity(Velocity=&Multiply_VectorByFloat(
     Vector=$Output.FixShapeVector.ShapeVector
     Float=&RandomRangeFloat(Minimum=5, Maximum=15, Evaluation Type=::NewEnumerator1)
    )) {module=AddVelocity001}
   /CascadeToNiagaraConverter/NiagaraScripts/ModuleScripts/AddVectorToPosition(Vector:Vector3f=&TransformVector(Vector=<0 0 0>, Source Space=::World, Destination Space=::Simulation))
  ]
  particleUpdateScript [
   &ParticleState()
   &SpriteRotationRate(Rotation Rate=&AngleConversion(
     Angle:NiagaraFloat=$Particles.SpriteRotationRate
     Angle Input:ENiagara_AngleInput=::NewEnumerator1
     Angle Output:ENiagara_AngleInput=::NewEnumerator0
    ))
   &ScaleSpriteSize(
    Scale Sprite Size Mode=::NewEnumerator0
    Uniform Scale Factor=&FloatFromCurve(FloatCurve=DataInterface() {propertyValues {Curve {keys [
         {arriveTangent=7.392, interpMode=RCIM_Cubic, leaveTangent=7.392, tangentMode=RCTM_User, time=0, value=0}
         {arriveTangent=2.448, interpMode=RCIM_Cubic, leaveTangent=2.448, tangentMode=RCTM_User, time=0.1188, value=0.6049}
         <at 0.6179 1>
        ]}}})
   )
   &Drag(Drag=3, Use Linear Drag=true, Use Rotational Drag=true)
   &AccelerationForce(Acceleration=$Particles.Acceleration)
   &SolveForcesAndVelocity(Clamp Velocity=false, Limit Acceleration=false)
   &SubUVAnimation(End Frame Range Override=58, Start Frame Range Override=20, UseEndFrame=false, UseStartFrame=false)
   &ScaleColor(
    Scale Alpha=&FloatFromCurve(FloatCurve=DataInterface() {propertyValues {Curve {keys [
         {arriveTangent=4.9805, interpMode=RCIM_Cubic, leaveTangent=4.9805, tangentMode=RCTM_User, time=0, value=0}
         <at 0.1798 1.0081 RCIM_Cubic>
         {arriveTangent=-2.9566, interpMode=RCIM_Cubic, leaveTangent=-2.9566, tangentMode=RCTM_User, time=1, value=0}
        ]}}})
    Scale RGB=&VectorFromCurve(VectorCurve=DataInterface() {propertyValues {
       XCurve {keys [{arriveTangent=-2.8741, time=0, value=0.2435}, <at 0.25 0.1015>, <at 0.8 0.0052>]}
       YCurve {keys [{arriveTangent=-1.0535, time=0, value=0.1373}, <at 0.25 0.0373>, <at 0.8 0.0025>]}
       ZCurve {keys [{arriveTangent=-19.89, time=0, value=0.5}, <at 0.25 0.18>, <at 0.8 0.05>]}
      }})
    ScaleA=true
    ScaleRGB=true
   )
   &DynamicMaterialParameters(
    Index 0 Param 1=&FloatFromCurve(FloatCurve=DataInterface() {propertyValues {Curve {keys [
         {arriveTangent=1.9398, time=0, value=-1}
         {arriveTangent=-0.5906, leaveTangent=-0.5906, time=0.0498, value=0.9398}
         {arriveTangent=-1.1585, interpMode=RCIM_Cubic, leaveTangent=-1.1585, time=0.2002, value=0.3492}
         {arriveTangent=-1.6869, interpMode=RCIM_Cubic, leaveTangent=-1.6869, time=0.6012, value=0.301}
         <at 1 -1 RCIM_Cubic>
        ]}}})
    Index 0 Param 2=1
    Index 0 Param 3=1
    Index 0 Param 4=&FloatFromCurve(FloatCurve=DataInterface() {propertyValues {Curve {keys [
         {arriveTangent=1.7442, interpMode=RCIM_Cubic, leaveTangent=1.7442, tangentMode=RCTM_User, time=0.1, value=0}
         {arriveTangent=2.8054, interpMode=RCIM_Cubic, leaveTangent=2.8054, time=0.3129, value=1.0643}
         <at 0.8129 2>
        ]}}})
    Module.Param0WriteEnabled=true
    Module.Param1WriteEnabled=true
    Module.Param2WriteEnabled=true
    Module.Param3WriteEnabled=true
   )
   &CameraOffset(Camera Offset Amount=30)
  ]
  rendererClasses [&NiagaraSpriteRendererProperties]
  renderers [{rendererClass=&NiagaraSpriteRendererProperties}]
 }
}
summary {
 systemName=NS_Husk_Damage_Pickaxe
 emitters [
  {emitterName=SmokeSwirl, rendererClasses [&NiagaraSpriteRendererProperties]}
  {emitterName=Base, rendererClasses [&NiagaraMeshRendererProperties]}
 ]
 userVariables []
}
