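# Translucent, lit particle material for death ash in the Husk enemy's assets, drawing tumbling dust-mote flakes.
# Per-particle random values select from a 4x2 SubUV sheet, tinted by particle color.
# Randomly offset noise plus particle alpha is clamped and multiplies the sheet alpha for uneven fading.
# A rotate-about-axis offset spins each particle around its center on one of three random axes, with angle driven by distance from world origin, so movement produces spin.
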Material /Game/Characters/Enemies/Husk/Materials/M_DeathAsh_01
graph {
 properties {
  bEnableResponsiveAA=true
  bUsedWithNanite=true
  bUsedWithParticleSprites=true
  blendMode=BLEND_Translucent
  opacityMaskClipValue=0.05
  pixelDepthOffsetMode=PDOM_Legacy
  translucencyLightingMode=TLM_VolumetricPerVertexNonDirectional
  usedShadingModels=Default Lit
 }
 nodes [
  n1 = ParticleRandom()
  n2 = MaterialFunctionCall!(
   U_Images=Constant() {r=4}
   V_Images=Constant() {r=2}
   SubUV Texture RGB=TextureObject() {texture=/Game/Effects/Fort_Effects/Textures/T_DustMotes_4x2_01}
   Time Speed
   Manual Time Control=StaticBool() {value=true}
   Time=Multiply(A=@n1) {constB=2}
   UVs
  ) {materialFunction=/Game/Effects/Fort_Effects/Materials/_Functions/MF_SubUV}
  n3 = ParticleColor()
  Multiply(A=@n2.RGB, B=@n3.RGB)
  n4 = TextureSample(UVs=Add(A=@n1, B=TextureCoordinate() {uTiling=0.25, vTiling=0.5})) {texture=&T_NoisePack_01}
  Multiply(A=@n2.A, B=Clamp(None=Add(A=@n4.R, B=@n3.A)))
  n5 = ParticleRandom()
  n6 = VectorParameter() {parameterName=NormalizeRotationAxis_1, defaultValue=<rgba 1 0.3 -0.2 0>}
  n7 = VectorParameter() {parameterName=NormalizeRotationAxis_2, defaultValue=<rgba -0.44 1 0.13 0>}
  n8 = VectorParameter() {parameterName=NormalizeRotationAxis_3, defaultValue=<rgba 0.06 -0.22 1 0>}
  n9 = Distance(A=Constant(), B=ParticlePositionWS())
  n10 = ScalarParameter() {defaultValue=500, parameterName=Spin Scale}
  n11 = ParticleRandom()
  RotateAboutAxis(
   NormalizedRotationAxis=Normalize(VectorInput=MaterialFunctionCall!(Random Picker=@n5, Texture Input, Option 1=@n6.RGB, Option 2=@n7.RGB, Option 3=@n8.RGB) {materialFunction=/Game/Effects/Fort_Effects/Materials/_Functions/MF_RGB_Random_01}.Vec3 Out Only)
   RotationAngle=Multiply(A=Divide(A=@n9, B=@n10), B=Clamp(None=@n11) {minDefault=0.55})
   PivotPoint=ParticlePositionWS()
   Position=WorldPosition().XYZ
  )
 ]
}
