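# Translucent, depth-test-disabled skinned-mesh material for Husk enemies, named as a building-hit effect, with a fresnel-driven orange-to-red tint.
# Opacity combines inverted fresnel, a timestamp in/out transition and a powered absolute scene-versus-pixel depth difference, fading it near scene depth.
# A custom stencil test zeroes opacity where the value equals 1.
# Separate chains compute a sine pulse and a normal-based vertex offset.
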Material /Game/Characters/Enemies/Husk/Materials/M_EnemyHitBuilding_01
graph {
 properties {
  bDisableDepthTest=true
  bEnableResponsiveAA=true
  bUsedWithMeshDeformer=true
  bUsedWithMorphTargets=true
  bUsedWithSkeletalMesh=true
  blendMode=BLEND_Translucent
  pixelDepthOffsetMode=PDOM_Legacy
  translucencyLightingMode=TLM_VolumetricPerVertexNonDirectional
  usedShadingModels=Default Lit
 }
 nodes [
  n1 = Add(A=Time(), B=ScalarParameter() {parameterName=Pulse Timestamp})
  n2 = ScalarParameter() {defaultValue=2, parameterName=Pulse Speed}
  ConstantBiasScale(None=Sine(None=Multiply(A=@n1, B=@n2))) {bias=1.6, scale=0.15}
  n3 = TextureSampleParameter2D() {
   parameterName=Diffuse
   samplerType=SAMPLERTYPE_Masks
   texture=/Game/Characters/Player/Female/Medium/Bodies/F_MED_Commando_01/Textures/SolidWhite_MaskCompression
  }
  n4 = VectorParameter() {parameterName=Color Low, defaultValue=<rgba 0.5 0.1324 0.0098 0>}
  n5 = VectorParameter() {parameterName=Color High, defaultValue=<rgba 1.2 0.06 0 0>}
  n6 = MaterialFunctionCall(
   Brightness=ScalarParameter() {defaultValue=1.25, parameterName=Fresnel Brightness}
   Exponent=ScalarParameter() {defaultValue=4.5, parameterName=Fresnel Exponent}
   Vertex / Pixel Normal=Transform(None=TextureSampleParameter2D() {parameterName=Normal Map, samplerType=SAMPLERTYPE_Normal, texture=&FlatNormalBlue}.RGB)
  ) {materialFunction=&MF_Fresnel_FX}
  LinearInterpolate(
   A=@n3.RGB
   B=Multiply(A=@n3.RGB, B=LinearInterpolate(A=@n4.RGB, B=@n5.RGB, Alpha=@n6.Fresnel Inverted))
   Alpha=ScalarParameter() {defaultValue=0.5, parameterName=Base Color Lerp}
  )
  Multiply(A=VertexNormalWS(), B=VectorParameter() {parameterName=VertOffset, defaultValue=<rgba 0.4 0.4 0.4 0>}.RGB)
  n7 = LinearInterpolate(
   A=ScalarParameter() {defaultValue=0.4, parameterName=Opacity Low}
   B=ScalarParameter() {defaultValue=1, parameterName=Opacity High}
   Alpha=@n6.Fresnel Inverted
  )
  n8 = MaterialFunctionCall(
   TimestampA=ScalarParameter() {parameterName=TimestampA}
   TransitionLength_In=ScalarParameter() {group=Timestamp, parameterName=TransitionLength_In}
   TimestampB=ScalarParameter() {defaultValue=1, parameterName=TimestampB}
   TransitionLength_Out=ScalarParameter() {defaultValue=-1, group=Timestamp, parameterName=TransitionLength_Out}
   Transition In Only?=StaticBool()
  ) {materialFunction=&MF_Timestamp_InOut_01}
  n9 = Subtract(A=SceneDepth().R, B=PixelDepth().R)
  n10 = Clamp(None=Divide(A=Abs(None=@n9), B=ScalarParameter() {defaultValue=70, parameterName=Depth Fade Divisor}))
  n11 = ScalarParameter() {defaultValue=10, parameterName=Depth Fade Power}
  n12 = ComponentMask(None=SceneTexture() {sceneTextureId=PPI_CustomStencil}.Color) {r=true}
  n13 = Constant() {r=1}
  n14 = Constant()
  n15 = Constant() {r=1}
  Multiply(
   A=Multiply(A=Multiply(A=@n7, B=@n8.Linear), B=Power(Base=@n10, Exp=@n11))
   B=OneMinus(None=If(A=@n12, B=@n13, "A > B"=@n14, "A == B"=@n15, "A < B"=@n14))
  )
 ]
}
