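# A library-exposed material function that applies elemental looks (fire, ice, lightning) to enemy character materials.
# It passes incoming material attributes through three chained static switch parameters, isFire, isIce and isLightning, each optionally applying that element's modifier function.
# Disabled switches pass attributes through unchanged, and enabled ones stack in that order, resolved at compile time.
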MaterialFunction /Game/Characters/Enemies/MaterialAttributes/ElementalSwitchMaterialAttribute
graph {
 properties {bExposeToLibrary=true, libraryCategoriesText [Fortnite Parent Materials]}
 nodes [
  n1 = FunctionInput(Preview=MaterialFunctionCall() {materialFunction=&/MaterialAttributeTester}.Result) {inputType=FunctionInput_MaterialAttributes}
  n2 = StaticSwitchParameter(
   True=MaterialFunctionCall!(MA=@n1) {materialFunction=/Game/Characters/Enemies/MaterialModifiers/Elemental_Fire/MA_Elemental_Fire}.Result
   False=@n1
  ) {group=ElementalSwitch, parameterName=isFire}
  n3 = StaticSwitchParameter(
   True=MaterialFunctionCall!(MA=@n2) {materialFunction=/Game/Characters/Enemies/MaterialModifiers/Elemental_Ice/MA_Elemental_Ice}.Result
   False=@n2
  ) {group=ElementalSwitch, parameterName=isIce}
  FunctionOutput(None=StaticSwitchParameter(
    True=MaterialFunctionCall!(MA=@n3) {materialFunction=/Game/Characters/Enemies/MaterialModifiers/Elemental_Lightning/MA_Elemental_Lightning}.Result
    False=@n3
   ) {group=ElementalSwitch, parameterName=isLightning})
 ]
}
