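# Material function that applies the final shared attribute pass to enemy character materials, exposed to the library under Fortnite Parent Materials.
# It takes a Modifier MA material-attributes input, previewed with BaseMaterialAttribute.
# That input passes in series through SpawnInEnemiesMaterialAttribute, EnemyGlowMaterialAttribute and DissolveEnemiesMaterialAttribute, layering spawn-in, glow and dissolve stages before output as MA Out.
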MaterialFunction /Game/Characters/Enemies/MaterialAttributes/EnemyFinalMaterialAttribute
graph {
 properties {bExposeToLibrary=true, libraryCategoriesText [Fortnite Parent Materials]}
 nodes [
  n1 = MaterialFunctionCall!(MA In=FunctionInput(Preview=MaterialFunctionCall() {materialFunction=&/BaseMaterialAttribute}.Result) {inputName=Modifier MA, inputType=FunctionInput_MaterialAttributes}) {materialFunction=&/SpawnInEnemiesMaterialAttribute}
  FunctionOutput(None=MaterialFunctionCall!(MA In=MaterialFunctionCall!(MA In=@n1.MA Out) {materialFunction=&/EnemyGlowMaterialAttribute}.Result) {materialFunction=&/DissolveEnemiesMaterialAttribute}.MA Out) {outputName=MA Out}
 ]
}
