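# EnemyMaterialAttribute is a material function that builds the shared material attributes for enemy characters, exposed to the library under Fortnite Parent Materials.
# It runs BaseMaterialAttribute and feeds its result into EnemyEyeGlowMaterialAttribute, which returns the final material attributes output.
# The function adds no logic of its own and only chains the two nested functions.
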MaterialFunction /Game/Characters/Enemies/MaterialAttributes/EnemyMaterialAttribute
graph {
 properties {bExposeToLibrary=true, libraryCategoriesText [Fortnite Parent Materials]}
 nodes [FunctionOutput(None=MaterialFunctionCall!(MA In=MaterialFunctionCall() {materialFunction=&/BaseMaterialAttribute}.Result) {materialFunction=&/EnemyEyeGlowMaterialAttribute}.MA Out)]
}
