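# Material function that adds an elemental emissive glow to the Mist Blaster enemy's incoming material attributes.
# The FX mask's body channel, inverted AO and a squared view-normal term that favors camera-facing surfaces shape the glow.
# MF_BlasterCommon supplies the color from lightning/ice/fire switches, scaled by a scalar input defaulting to zero.
# Scrolling camera-aligned noise drives a color lerp, but both ends use DarkColor, so it has no visible effect.
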MaterialFunction /Game/Characters/Enemies/Mist_Blaster/Materials/MF_MistBlasterFX
graph {nodes [
  n1 = GetMaterialAttributes(None=FunctionInput() {inputName=InMA, inputType=FunctionInput_MaterialAttributes}) {attributeGetTypes [#1, #2]}
  n2 = GetMaterialAttributes(None=@n1.MaterialAttributes) {attributeGetTypes [#3]}
  n3 = TextureSampleParameter2D() {
   desc="R = body\r\nG = chest hot glow\r\nB = eyes and mouth"
   group=Deimos
   parameterName=FX Mask
   samplerSource=SSM_Wrap_WorldGroupSettings
   samplerType=SAMPLERTYPE_Masks
   texture=/Game/Characters/Enemies/Mist_Blaster/Textures/T_Mist_Blaster_Body_FX
  }
  n4 = DotProduct(A=CameraVectorWS(), B=Transform(None=@n1.Normal))
  n5 = Multiply(A=Multiply(A=OneMinus(None=@n1.AmbientOcclusion), B=@n3.R), B=LinearInterpolate(Alpha=Multiply(A=@n4, B=@n4)) {constA=0.25})
  n6 = Constant() {r=0.2}
  n7 = StaticSwitchParameter(True=Constant() {r=0.1}, False=Constant()) {group=ElementalSwitch, parameterName=isFire}
  n8 = MaterialFunctionCall!(
   TimeScale
   ElementColor=StaticSwitchParameter(True=Constant() {r=0.3}, False=StaticSwitchParameter(True=@n6, False=@n7) {group=ElementalSwitch, parameterName=isIce}) {group=ElementalSwitch, parameterName=isLightning}
   HeadEnhanceScale
  ) {materialFunction=/Game/Effects/Fort_Effects/Materials/_Functions/MF_BlasterCommon}
  n9 = TransformPosition(None=WorldPosition().XYZ) {transformSourceType=TRANSFORMPOSSOURCE_World}
  n10 = Transform(None=Constant3Vector() {constant=<rgba 1 0 0 0>}) {transformSourceType=TRANSFORMSOURCE_Camera}
  n11 = Transform(None=Constant3Vector() {constant=<rgba 0 1 0 0>}) {transformSourceType=TRANSFORMSOURCE_Camera}
  n12 = Divide(A=AppendVector(A=DotProduct(A=@n9, B=@n10), B=DotProduct(A=@n9, B=@n11))) {constB=128}
  n13 = Multiply(A=Time()) {constB=0.25}
  n14 = Multiply(
   A=Multiply(A=@n5, B=@n5)
   B=TextureSample(UVs=Subtract(A=@n12, B=@n13)) {samplerType=SAMPLERTYPE_LinearColor, texture=&T_NoisePack_LinearColor}.R
  )
  n15 = Constant()
  n16 = Multiply(
   A=@n5
   B=Multiply(
    A=LinearInterpolate(A=@n8.DarkColor, B=@n8.DarkColor, Alpha=@n14)
    B=FunctionInput(Preview=@n15) {bUsePreviewValueAsDefault=true, inputName=InMA1, inputType=FunctionInput_Scalar, sortPriority=1}
   ) {constB=10}
  )
  FunctionOutput(None=SetMaterialAttributes!(MaterialAttributes=@n2.MaterialAttributes, EmissiveColor=Add(A=@n2.EmissiveColor, B=@n16)) {attributeSetTypes [#3]}) {outputName=ResultMA}
 ]}
