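# Parent masked subsurface-profile material for skeletal enemy meshes such as Husks, using material attributes, Nanite support and a Husk flesh physical material.
# Three chained material functions build the attributes: a base enemy layer, an elemental switch, then a final enemy pass.
# A static switch, ForMistBlaster, optionally routes the result through the Mist Blaster FX function, driven by a throat emissive parameter.
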Material /Game/Characters/Enemies/Husk/Materials/M_Enemy_Parent_Material
graph {
 properties {
  bUseMaterialAttributes=true
  bUsedWithClothing=true
  bUsedWithMeshDeformer=true
  bUsedWithMorphTargets=true
  bUsedWithNanite=true
  bUsedWithSkeletalMesh=true
  bUsedWithStaticLighting=true
  blendMode=BLEND_Masked
  numCustomizedUVs=4
  opacityMaskClipValue=0.35
  physMaterial=/Game/FortressPhysicalMaterials/Husk_Flesh
  pixelDepthOffsetMode=PDOM_Legacy
  shadingModel=MSM_SubsurfaceProfile
  subsurfaceProfile=/Game/Packages/Fortress_SharedMaterials/SubSurfaceProfile/SS_MistMonster_01
  usedShadingModels=Subsurface Profile
 }
 nodes [
  n1 = MaterialFunctionCall!(Modifier MA=MaterialFunctionCall!(In=MaterialFunctionCall() {materialFunction=/Game/Characters/Enemies/MaterialAttributes/EnemyMaterialAttribute}.Result) {materialFunction=/Game/Characters/Enemies/MaterialAttributes/ElementalSwitchMaterialAttribute}.Result) {materialFunction=/Game/Characters/Enemies/MaterialAttributes/EnemyFinalMaterialAttribute}
  StaticSwitchParameter(
   True=MaterialFunctionCall!(InMA=@n1.MA Out, InMA1=ScalarParameter() {defaultValue=12, parameterName=Blaster_ThroatEmissive}) {materialFunction=/Game/Characters/Enemies/Mist_Blaster/Materials/MF_MistBlasterFX}.ResultMA
   False=@n1.MA Out
  ) {parameterName=ForMistBlaster}
 ]
}
