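# Unlit, two-sided translucent mesh-particle material for the Husk melee damage swoop effect.
# A rotatable swoop texture supplies green for particle-colored emissive and red for the erosion mask.
# Panning noise, weighted by a V gradient, erodes that mask over particle lifetime through a hard Floor threshold.
# Opacity combines the mask with particle alpha, depth fade (bypassed on mobile) and a V gradient offset by the PanMaskX dynamic parameter.
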Material /Game/Characters/Enemies/Husk/Materials/M_HuskDamageMeleeSwoop
graph {
 properties {
  bUsedWithMeshParticles=true
  blendMode=BLEND_Translucent
  pixelDepthOffsetMode=PDOM_Legacy
  shadingModel=MSM_Unlit
  twoSided=true
  usedShadingModels=Unlit
 }
 nodes [
  n1 = ParticleColor()
  n2 = TextureSample(UVs=MaterialFunctionCall(
    UVs=TextureCoordinate()
    Rotation Center=Constant2Vector() {g=0.5, r=0.5}
    "Rotation Angle (0-1)"=ScalarParameter() {parameterName=RotationAngle}
   ) {materialFunction=&CustomRotator}.Rotated Values) {texture=/Game/Effects/Fort_Effects/Textures/Husk/T_HuskMeleeSwoop}
  MaterialFunctionCall(Emissive=Multiply(A=@n2.G, B=@n1.RGB)) {materialFunction=&MF_FX_EmissiveAdjustments}
  n3 = Panner(Coordinate=TextureCoordinate() {uTiling=3}) {speedY=-0.5}
  n4 = OneMinus(None=ComponentMask(None=TextureCoordinate()) {g=true})
  n5 = Multiply(
   A=Multiply(A=TextureSample(UVs=@n3) {texture=&T_NoisePack_01}.R, B=Power(Base=@n4))
   B=ScalarParameter() {defaultValue=3, parameterName=ErodeMultiplier}
  ) {constB=3}
  n6 = OneMinus(None=ParticleRelativeTime())
  n7 = Floor(None=Add(A=Subtract(A=@n2.R, B=@n5), B=Clamp(None=@n6) {maxDefault=0.99, minDefault=0.01}))
  n8 = DepthFade(FadeDistance=Constant() {r=12})
  n9 = Constant() {r=1}
  n10 = TextureCoordinate()
  Multiply(
   A=Multiply(A=Multiply(A=@n1.A, B=@n7), B=ShadingPathSwitch(Default=@n8, Deferred=@n8, Mobile=@n9) {desc=Migrated from FeatureLevelSwitch})
   B=Add(A=ComponentMask(None=@n10) {g=true}, B=DynamicParameter() {defaultValue=<rgba 0 1 1 1>, paramNames [PanMaskX,Param2,Param3,Param4]}.PanMaskX) {constB=0.5}
  )
 ]
}
