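# Unlit translucent particle sprite material with responsive AA, named for Husk Lobber GPU particles, that draws animated bits from an 8x8 SubUV texture.
# Particle relative time plus a per-particle random offset, scaled by a Frame Rate parameter, picks the flipbook frame.
# The summed RGB channels tinted by particle color feed the MF_FX_EmissiveAdjustments function, while texture alpha is multiplied by particle alpha.
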Material /Game/Characters/Enemies/Husk/Materials/M_HuskLobber_GPU_Particles
graph {
 properties {
  bEnableResponsiveAA=true
  bUsedWithParticleSprites=true
  blendMode=BLEND_Translucent
  pixelDepthOffsetMode=PDOM_Legacy
  shadingModel=MSM_Unlit
  usedShadingModels=Unlit
 }
 nodes [
  n1 = ParticleColor()
  n2 = ParticleRelativeTime()
  n3 = ParticleRandom()
  n4 = MaterialFunctionCall!(
   Texture=TextureObject() {texture=/Game/Effects/Fort_Effects/Textures/Husk/T_Husk_Bits_SubUV}
   UVs
   SubImages=Constant2Vector() {g=8, r=8}
   Frame=Multiply(A=Add(A=@n2, B=@n3), B=ScalarParameter() {defaultValue=64, parameterName=Frame Rate})
  ) {materialFunction=/Engine/Functions/Engine_MaterialFunctions03/Texturing/SubUV_Function}
  n5 = ComponentMask(None=@n4.RGB) {r=true}
  n6 = Add(A=ComponentMask(None=@n4.RGB) {b=true}, B=ComponentMask(None=@n4.RGB) {g=true})
  MaterialFunctionCall(Emissive=Multiply(A=@n1.RGB, B=Add(A=@n5, B=@n6))) {materialFunction=&MF_FX_EmissiveAdjustments}
  Multiply(A=@n1.A, B=@n4.Alpha)
 ]
}
