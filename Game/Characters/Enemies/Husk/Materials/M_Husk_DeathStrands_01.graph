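# Unlit translucent mesh-particle material for the Husk enemy's death strands effect.
# A vertically panned tiling texture plus a bias is masked by Fresnel, the UV V gradient and vertex alpha, then scaled by particle alpha as an opacity-style mask.
# Particle color feeds an emissive adjustment function.
# Dynamic parameters control V tiling, pan speed, bias and a thickness offset along vertex normals.
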Material /Game/Characters/Enemies/Husk/Materials/M_Husk_DeathStrands_01
graph {
 properties {
  bEnableResponsiveAA=true
  bUsedWithMeshParticles=true
  blendMode=BLEND_Translucent
  pixelDepthOffsetMode=PDOM_Legacy
  shadingModel=MSM_Unlit
  usedShadingModels=Unlit
 }
 nodes [
  n1 = ParticleColor()
  MaterialFunctionCall(Emissive=@n1.RGB) {materialFunction=&MF_FX_EmissiveAdjustments}
  n2 = Multiply(
   A=MaterialFunctionCall(Exponent=ScalarParameter() {defaultValue=5, parameterName=Fresnel Exponent}) {materialFunction=&MF_Fresnel_FX}.Fresnel
   B=ComponentMask(None=TextureCoordinate()) {g=true}
  )
  n3 = DynamicParameter() {paramNames [V Scale, V Panner Speed, Bias, Thickness]}
  n4 = Multiply(A=AppendVector(A=Constant() {r=1}, B=@n3.V Scale), B=Add(A=@n3.V Panner Speed, B=TextureCoordinate()))
  n5 = Multiply(A=@n3.V Panner Speed, B=Time())
  n6 = Add(
   A=TextureSample(UVs=Panner(Coordinate=@n4, Time=@n5) {speedY=1}) {samplerType=SAMPLERTYPE_LinearColor, texture=/Game/Effects/Fort_Effects/Textures/Tile/T_Raygun_Muzzle_02}.R
   B=DynamicParameter() {paramNames [V Scale, V Panner Speed, Bias, Thickness]}.Bias
  )
  Multiply(A=@n1.A, B=Multiply(A=Multiply(A=@n2, B=@n6), B=VertexColor().A))
  Multiply(A=DynamicParameter() {paramNames [V Scale, V Panner Speed, Bias, Thickness]}.Thickness, B=VertexNormalWS())
 ]
}
