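# Default Lit material for the Husk enemy's skeleton, flagged for skeletal meshes, morph targets, mesh deformers and mesh particles.
# The diffuse texture and the specular texture, scaled down by a small constant, feed ConvertFromDiffSpec to translate diffuse/specular inputs to PBR values.
# A separate sample reads the normal map.
# The specular texture also uses the Normal sampler type.
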Material /Game/Characters/Enemies/Husk/Materials/M_Husk_Skeleton
graph {
 properties {
  bUsedWithMeshDeformer=true
  bUsedWithMeshParticles=true
  bUsedWithMorphTargets=true
  bUsedWithSkeletalMesh=true
  pixelDepthOffsetMode=PDOM_Legacy
  usedShadingModels=Default Lit
 }
 nodes [
  MaterialFunctionCall(
   DiffuseColor=TextureSample() {texture=/Game/Characters/Enemies/Husk/Textures/T_Skeleton_D}.RGB
   SpecularColor=Multiply(A=TextureSample() {samplerType=SAMPLERTYPE_Normal, texture=/Game/Characters/Enemies/Husk/Textures/T_Skeleton_S}.RGB) {constA=1, constB=0.04, desc=FresnelBaseReflectFraction}
  ) {materialFunction=&ConvertFromDiffSpec}
  TextureSample() {samplerType=SAMPLERTYPE_Normal, texture=/Game/Characters/Enemies/Husk/Textures/T_Skeleton_N}
 ]
}
