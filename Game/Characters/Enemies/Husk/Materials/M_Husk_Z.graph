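# Unlit translucent particle-sprite material for the Husk "Z" effect, with warm and dark tones and a swirling dissipation.
# A texture-driven UV swirl scaled by dynamic DistortionAmount warps the Husk Z texture.
# Distance blending switches to a no-mip texture, shifts distortion to a fixed amount and pushes emissive toward the warm color.
# Opacity combines a HeightLerp fade driven by particle Alpha, TransparencyMultiplier, distortion erosion and depth-based particle opacity, then is squared and masked.
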Material /Game/Characters/Enemies/Husk/Materials/M_Husk_Z
graph {
 properties {
  bUsedWithParticleSprites=true
  blendMode=BLEND_Translucent
  pixelDepthOffsetMode=PDOM_Legacy
  shadingModel=MSM_Unlit
  usedShadingModels=Unlit
 }
 nodes [
  n1 = Constant3Vector() {desc=dissapation color, constant=<rgba 0.012 0.01 0.01 0>}
  n2 = Constant3Vector() {constant=<rgba 1.5 1 0.5 0>}
  n3 = LinearInterpolate(A=@n1, B=@n2)
  n4 = TextureCoordinate()
  n5 = TextureSample(UVs=@n4) {samplerType=SAMPLERTYPE_LinearColor, texture=/Game/Characters/Enemies/Husk_Deimos/Textures/Plane001NewStuffs}
  n6 = DynamicParameter() {defaultValue=<rgba 0 0 0 1>, paramNames [Frame Picker, DistortionAmount, TransparencyMultiplier, Alpha]}
  n7 = Constant() {r=800}
  n8 = Constant() {r=-1500}
  n9 = MaterialFunctionCall!(Blend Range=@n7, Start Offset=@n8) {materialFunction=/Engine/Functions/Engine_MaterialFunctions03/Particles/Distance_Blend}
  n10 = Multiply(
   A=Subtract(A=@n4, B=ComponentMask(None=@n5.RGB) {g=true, r=true})
   B=LinearInterpolate(A=@n6.DistortionAmount, B=Constant() {r=0.3}, Alpha=@n9.Result)
  ) {desc=Swirl * dissapation}
  n11 = Add(A=@n4, B=@n10)
  n12 = TextureSample(UVs=@n11) {samplerType=SAMPLERTYPE_LinearColor, texture=/Game/Characters/Enemies/Husk_Deimos/Textures/T_Husk_Z}
  n13 = TextureSample(UVs=@n11) {samplerType=SAMPLERTYPE_LinearColor, texture=/Game/Characters/Enemies/Husk_Deimos/Textures/T_Husk_Z_No_Mip_ForDistance}
  n14 = LinearInterpolate(A=@n12.A, B=@n13.A, Alpha=@n9.Result)
  MaterialFunctionCall(Emissive=LinearInterpolate(
    A=LinearInterpolate(A=@n3, B=@n1, Alpha=@n14)
    B=@n2
    Alpha=MaterialFunctionCall(Blend Range=@n7, Start Offset=@n8) {materialFunction=&Distance_Blend}.Result
   )) {materialFunction=&MF_FX_EmissiveAdjustments}
  n15 = ComponentMask(None=LinearInterpolate(A=@n12.B, B=@n13.B, Alpha=@n9.Result)) {r=true}
  n16 = ScalarParameter() {defaultValue=1, parameterName=DepthFalloff}
  n17 = ComponentMask(None=LinearInterpolate(A=@n12.R, B=@n13.R, Alpha=@n9.Result)) {r=true}
  n18 = Multiply(A=@n15, B=Constant3Vector() {constant=<rgba 1 1 1 0>})
  n19 = OneMinus(None=@n15)
  n20 = Constant() {r=2}
  n21 = Multiply(
   A=@n6.TransparencyMultiplier
   B=OneMinus(None=MaterialFunctionCall(A=@n17, B=@n18, Transition Phase=@n6.Alpha, Height Texture=@n19, Contrast=@n20) {materialFunction=&HeightLerp}.Results)
  ) {desc=Fade in}
  n22 = Abs(None=@n10)
  n23 = Multiply(A=Add(A=ComponentMask(None=@n22) {r=true}, B=ComponentMask(None=@n22) {g=true})) {constB=10}
  n24 = Clamp(None=Multiply(
    A=MaterialFunctionCall!(
     Depth Texture=@n15
     -------------Depth Texture Options-----------
     "Use dynamic (True) or explicit texture depth settings (False)"
     Dynamic Texture Depth Ratio=@n16
     "Explicit Texture Depth (WS)"
     "Depth Texture Falloff Softness (WS) (1/n)"
     -------Camera Falloff Group-------
     Use Near Camera Falloff
     Near Camera Falloff Start Distance
     "Near Camera Fade Distance (1/n)"
     Use Depth In Camera Falloff Calculation
     -------Particle Alpha Group-------
     Use Particle Alpha
     -------WPO Group-------
     Camera Offset
    ) {materialFunction=/Engine/Functions/Engine_MaterialFunctions02/Texturing/3dParticleOpacity}.Opacity
    B=Subtract(A=@n21, B=@n23)
   ))
  Multiply(
   A=Multiply(A=@n24, B=@n24)
   B=TextureSample() {samplerType=SAMPLERTYPE_LinearColor, texture=&Plane001NewStuffs}.B
  )
 ]
}
