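# Translucent unlit smoke trail material for the Lobber Husk, used on ribbons and beam trails.
# A scrolling mask distorts a soft band across the trail, combined with panning noise to darken particle colour and shape opacity with end and depth fades.
# Noise also drives a Z-axis offset scaled by particle alpha.
# Dynamic parameters control distortion, colour and levels, though the levels divisor subtracts WhitePoint from itself, yielding zero.
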Material /Game/Characters/Enemies/Husk/Materials/M_Lobber_Trail
graph {
 properties {
  bEnableResponsiveAA=true
  bUsedWithBeamTrails=true
  bUsedWithNiagaraRibbons=true
  blendMode=BLEND_Translucent
  pixelDepthOffsetMode=PDOM_Legacy
  shadingModel=MSM_Unlit
  usedShadingModels=Unlit
 }
 nodes [
  n1 = ParticleColor()
  n2 = DynamicParameter() {defaultValue=<rgba 1 0 0 0>, paramNames [WhitePoint, BlackPoint, DistortStrength, ColourStrength]}
  n3 = TextureCoordinate() {coordinateIndex=1, vTiling=0}
  n4 = Time()
  n5 = MaterialFunctionCall(X=Add(A=ComponentMask(None=@n3) {r=true}, B=Multiply(A=@n4) {constB=0.35}), Y=Constant() {r=0.5}) {materialFunction=&MakeFloat2}
  n6 = Add(
   A=ComponentMask(None=TextureCoordinate()) {g=true}
   B=Multiply(
    A=TextureSample(UVs=@n5.Result) {texture=/Game/Effects/Fort_Effects/Textures/Smoke/T_Lobber_Trail_Mask}.R
    B=DynamicParameter() {defaultValue=<rgba 1 0 0 0>, paramNames [WhitePoint, BlackPoint, DistortStrength, ColourStrength]}.DistortStrength
   ) {constB=0.1}
  ) {constB=0}
  n7 = Multiply(A=Multiply(A=@n6, B=@n6)) {constB=-11.1111}
  n8 = Multiply(A=@n6) {constB=11.1111}
  n9 = Saturate(None=Add(A=Add(A=@n7, B=@n8)) {constB=-1.7778})
  n10 = Multiply(A=@n2.ColourStrength, B=@n9)
  n11 = TextureCoordinate() {coordinateIndex=1}
  n12 = MaterialFunctionCall(X=ScalarParameter() {defaultValue=2, parameterName=UTile}, Y=ScalarParameter() {defaultValue=0.7, parameterName=VTile}) {materialFunction=&MakeFloat2}
  n13 = TextureSample(UVs=Panner(Coordinate=Multiply(A=@n11, B=@n12.Result)) {speedX=0.3811}) {texture=&T_NoisePack_01}
  n14 = Saturate(None=ConstantBiasScale(None=@n13.R) {bias=0.2, scale=0.8333})
  n15 = Saturate(None=Multiply(A=@n14) {constB=1.5})
  MaterialFunctionCall(Emissive=LinearInterpolate(A=@n1.RGB, B=Multiply(A=@n1.RGB) {constB=0.35}, Alpha=Multiply(A=@n10, B=@n15))) {materialFunction=&MF_FX_EmissiveAdjustments}
  n16 = OneMinus(None=ParticleColor().A)
  n17 = Multiply(A=Subtract(A=@n13.R) {constB=0.5}, B=LinearInterpolate(Alpha=Power(Base=@n16) {constExponent=1.35}) {constB=0.6}) {constB=0.37}
  Multiply(A=LinearInterpolate(Alpha=Add(A=@n17) {constB=0.5}) {constA=-184, constB=184}, B=Constant3Vector() {constant=<rgba 0 0 1 0>})
  n18 = Power(Base=@n9) {constExponent=3}
  n19 = Multiply(A=Multiply(A=Power(Base=@n9) {constExponent=16}) {constB=0.67}, B=@n2.ColourStrength)
  n20 = TextureCoordinate()
  n21 = Subtract(A=Multiply(A=ComponentMask(None=@n20) {r=true}) {constB=10}) {constB=0.05}
  n22 = ComponentMask(None=TextureCoordinate()) {r=true}
  n23 = Multiply(A=Multiply(A=Subtract(A=@n18, B=@n19) {constB=0}, B=@n14), B=Multiply(A=Saturate(None=@n21), B=OneMinus(None=@n22)))
  n24 = DynamicParameter() {defaultValue=<rgba 1 0 0 0>, paramNames [WhitePoint, BlackPoint, DistortStrength, ColourStrength]}
  n25 = Saturate(None=Divide(A=Subtract(A=@n23, B=@n24.BlackPoint), B=Subtract(A=@n24.WhitePoint, B=@n24.WhitePoint)))
  n26 = Constant() {r=16}
  Multiply(
   A=Multiply(A=Multiply(A=@n1.A, B=@n25) {constA=1}, B=DepthFade(FadeDistance=@n26))
   B=ScalarParameter() {defaultValue=0.5, parameterName=Opacity}
  )
 ]
}
