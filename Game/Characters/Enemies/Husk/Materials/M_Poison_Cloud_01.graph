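# Unlit translucent poison cloud material for Husk enemy effects on Niagara mesh particles.
# Opacity multiplies particle alpha by a powered, inverted Fresnel to soften mesh edges, then applies camera-distance blending and depth fade.
# Emissive fakes lighting by dotting the view-space normal with a constant light vector.
# That highlight is tinted by particle color plus a day/night color from the time-of-day parameter collection, then added to particle color through MF_FX_EmissiveAdjustments.
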Material /Game/Characters/Enemies/Husk/Materials/M_Poison_Cloud_01
graph {
 properties {
  bTangentSpaceNormal=false
  bUsedWithNiagaraMeshParticles=true
  blendMode=BLEND_Translucent
  pixelDepthOffsetMode=PDOM_Legacy
  shadingModel=MSM_Unlit
  translucencyPass=MTP_BeforeDOF
  usedShadingModels=Unlit
 }
 nodes [
  n1 = ParticleColor()
  n2 = StaticBool() {value=true}
  n3 = Clamp(None=Power(Base=MaterialFunctionCall(Invert Fresnel=@n2, Use Cheap contrast=@n2) {materialFunction=&Fresnel_Function}.Result) {constExponent=4})
  n4 = Constant() {r=100}
  n5 = Constant() {r=-50}
  DepthFade(Opacity=LinearInterpolate(
    B=Multiply(A=@n1.A, B=@n3)
    Alpha=MaterialFunctionCall!(Blend Range=@n4, Start Offset=@n5) {materialFunction=/Engine/Functions/Engine_MaterialFunctions03/Particles/Distance_Blend}.Result
   )) {fadeDistanceDefault=80}
  n6 = Normalize(VectorInput=Transform(None=Constant3Vector() {constant=<rgba 0 0 1 0>}) {transformType=TRANSFORM_View})
  n7 = Constant3Vector() {desc=Control Light angle, constant=<rgba 0 -1.2 0.4 0>}
  n8 = Power(Base=Multiply(A=DotProduct(A=@n6, B=@n7), B=Constant() {r=1})) {constExponent=5}
  n9 = Clamp(None=Multiply(A=Multiply(A=@n8) {constB=2}) {constB=0.5})
  n10 = CollectionParameter() {
   collection=&FortniteMaterialParameters
   parameterName=FortniteActiveTimeOfDay
  }
  n11 = Constant4Vector() {constant=<rgba 1 1 1 0>}
  n12 = Add(
   A=@n1.RGB
   B=LinearInterpolate(
    A=Constant3Vector() {constant=<rgba 0.4088 0.639 0.73 0>}
    B=Constant3Vector() {constant=<rgba 0.5 0.4875 0.2323 0>}
    Alpha=DotProduct(A=@n10, B=@n11.RGBA)
   ) {desc="A=Night B=Day"}
  )
  MaterialFunctionCall(Emissive=Add(A=Multiply(A=@n9, B=@n12), B=@n1.RGB)) {materialFunction=&MF_FX_EmissiveAdjustments}
 ]
}
